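# Animation Blueprint for the WaltzScout backpack cosmetic, adding secondary physics motion to its handle, strap, two axes and three charm bones.
# The cached input pose feeds two parallel Anim Dynamics chains over the same bones with different tuning, with springs on handle and axes and root-relative simulation for charms.
# A bool blend on bIsInFrontEnd outputs the second chain in the front end and the first otherwise.
# The event graph holds no logic.
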AnimBlueprint /BRCosmetics/Accessories/FORT_Backpacks/Backpack_F_MED_WaltzScout/Meshes/F_MED_WaltzScout_Pack_AnimBP
target_skeleton=&/F_MED_WaltzScout_Pack_Skeleton
graphs [
 AnimGraph {nodes [
   n1 = Anim Dynamics - Bone: dyn_bag_handle_jnt(ComponentPose=Local To Component(LocalPose=Use cached pose 'InputPoseSaved'()), Alpha=1, ExternalForce="0,0,0") {
    angularDampingOverride=0.9999
    angularSpringConstant=5000
    bAngularSpring=true
    bOverrideAngularDamping=true
    bUseGravityOverride=true
    bUsePlanarLimit=false
    alphaBoolBlend {blendOption=Linear}
    boundBone {boneName=dyn_bag_handle_jnt}
    gravityOverride {z=-980}
    physicsBodyDefinitions [{
      boundBone {boneName=dyn_bag_handle_jnt}
      boxExtents=<20 20 10>
      constraintSetup {angularLimitsMax=<5 0 20>, angularLimitsMin=<-5 0 -20>, angularTarget=<1 0 0>}
      localJointOffset=<-3 0 0>
     }]
   }
   n2 = Anim Dynamics - Bone: dyn_bag_strap_jnt(ComponentPose=@n1, Alpha=1, ExternalForce="0,0,0") {
    angularDampingOverride=0.999
    bOverrideAngularDamping=true
    bUseGravityOverride=true
    alphaBoolBlend {blendOption=Linear}
    boundBone {boneName=dyn_bag_strap_jnt}
    componentAppliedLinearAccClamp=<2000 2000 2000>
    componentLinearAccScale=<0.3 0.3 0.3>
    componentLinearVelScale=<0.3 0.3 0.3>
    gravityOverride {z=-100}
    physicsBodyDefinitions [{
      boundBone {boneName=dyn_bag_strap_jnt}
      boxExtents=<19 10 10>
      constraintSetup {angularLimitsMax=<0 8 10>, angularLimitsMin=<0 -8 -50>, angularTarget=<1 -0.5 0>}
      localJointOffset=<1.5 0.5 0>
     }]
   }
   n3 = Anim Dynamics - Bone: dyn_axe_l_jnt(ComponentPose=@n2, Alpha=1, ExternalForce="0,0,0") {
    angularDampingOverride=0.988
    angularSpringConstant=2000
    bAngularSpring=true
    bOverrideAngularDamping=true
    bUseGravityOverride=true
    gravityScale=0.35
    alphaBoolBlend {blendOption=Linear}
    boundBone {boneName=dyn_axe_l_jnt}
    componentAppliedLinearAccClamp=<200 200 200>
    componentLinearAccScale=<0.4 0.4 0.4>
    componentLinearVelScale=<0.5 0.5 0.5>
    gravityOverride {z=-300}
    physicsBodyDefinitions [{
      boundBone {boneName=dyn_axe_l_jnt}
      boxExtents=<20 20 18>
      constraintSetup {angularLimitsMax=<0 0 15>, angularLimitsMin=<0 0 -15>, angularTarget=<1 0 0>}
      localJointOffset=<-2 -0.4 0>
     }]
   }
   n4 = Anim Dynamics - Bone: dyn_axe_r_jnt(ComponentPose=@n3, Alpha=1, ExternalForce="0,0,0") {
    angularDampingOverride=0.988
    angularSpringConstant=1850
    bAngularSpring=true
    bOverrideAngularDamping=true
    bUseGravityOverride=true
    gravityScale=0.35
    alphaBoolBlend {blendOption=Linear}
    boundBone {boneName=dyn_axe_r_jnt}
    componentAppliedLinearAccClamp=<200 200 200>
    componentLinearAccScale=<0.4 0.4 0.4>
    componentLinearVelScale=<0.5 0.5 0.5>
    gravityOverride {z=-300}
    physicsBodyDefinitions [{
      boundBone {boneName=dyn_axe_r_jnt}
      boxExtents=<20 15 15>
      constraintSetup {angularLimitsMax=<0 0 15>, angularLimitsMin=<0 0 -15>, angularTarget=<1 0 0>}
      localJointOffset=<2 0 0>
     }]
   }
   n5 = Anim Dynamics - Bone: dyn_charm_01_jnt(ComponentPose=@n4, Alpha=1, ExternalForce="0,0,0") {
    angularDampingOverride=0.9998
    bOverrideAngularDamping=true
    bUseGravityOverride=true
    simulationSpace=RootRelative
    boundBone {boneName=dyn_charm_01_jnt}
    componentAppliedLinearAccClamp=<2000 2000 2000>
    componentLinearAccScale=<0.5 0.5 0.5>
    componentLinearVelScale=<0.5 0.5 0.5>
    gravityOverride {z=-600}
    physicsBodyDefinitions [{
      boundBone {boneName=dyn_charm_01_jnt}
      boxExtents=<8 8 8>
      constraintSetup {angularLimitsMax=<10 3 30>, angularLimitsMin=<-10 -10 -80>, angularTarget=<0 0 0>}
      localJointOffset=<1 0 0>
     }]
    relativeSpaceBone {boneName=dyn_axe_r_jnt}
   }
   n6 = Anim Dynamics - Bone: dyn_charm_02_jnt(ComponentPose=@n5, Alpha=1, ExternalForce="0,0,0") {
    angularDampingOverride=0.998
    bOverrideAngularDamping=true
    bUseGravityOverride=true
    simulationSpace=RootRelative
    boundBone {boneName=dyn_charm_02_jnt}
    componentAppliedLinearAccClamp=<2000 2000 2000>
    componentLinearAccScale=<0.5 0.5 0.5>
    componentLinearVelScale=<0.5 0.5 0.5>
    gravityOverride {z=-600}
    physicsBodyDefinitions [{
      boundBone {boneName=dyn_charm_02_jnt}
      boxExtents=<8 8 8>
      constraintSetup {angularLimitsMax=<10 20 20>, angularLimitsMin=<-10 -20 -80>, angularTarget=<2 0 0>}
      localJointOffset=<1 0 0>
     }]
    relativeSpaceBone {boneName=dyn_axe_r_jnt}
   }
   n7 = Anim Dynamics - Bone: dyn_charm_03_jnt(ComponentPose=@n6, Alpha=1, ExternalForce="0,0,0") {
    angularDampingOverride=0.999
    bOverrideAngularDamping=true
    bUseGravityOverride=true
    simulationSpace=RootRelative
    boundBone {boneName=dyn_charm_03_jnt}
    componentLinearAccScale=<0.5 0.5 0.5>
    componentLinearVelScale=<0.5 0.5 0.5>
    gravityOverride {z=-600}
    physicsBodyDefinitions [{
      boundBone {boneName=dyn_charm_03_jnt}
      boxExtents=<5 5 5>
      constraintSetup {angularLimitsMax=<10 40 0>, angularLimitsMin=<-10 -40 0>, angularTarget=<0 0 0>}
      localJointOffset=<1 0 0>
     }]
    relativeSpaceBone {boneName=dyn_axe_r_jnt}
   }
   n8 = Anim Dynamics - Bone: dyn_bag_handle_jnt(ComponentPose=Local To Component(LocalPose=Use cached pose 'InputPoseSaved'()), Alpha=1, ExternalForce="0,0,0") {
    angularDampingOverride=0.9
    angularSpringConstant=1000
    bAngularSpring=true
    bOverrideAngularDamping=true
    bUseGravityOverride=true
    bUsePlanarLimit=false
    alphaBoolBlend {blendOption=Linear}
    boundBone {boneName=dyn_bag_handle_jnt}
    gravityOverride {z=-500}
    physicsBodyDefinitions [{
      boundBone {boneName=dyn_bag_handle_jnt}
      boxExtents=<10 10 10>
      constraintSetup {angularLimitsMax=<5 0 20>, angularLimitsMin=<-5 0 -20>, angularTarget=<1 0 0>}
      localJointOffset=<1 0 0>
     }]
   }
   n9 = Anim Dynamics - Bone: dyn_bag_strap_jnt(ComponentPose=@n8, Alpha=1, ExternalForce="0,0,0") {
    angularDampingOverride=0.999
    bOverrideAngularDamping=true
    bUseGravityOverride=true
    alphaBoolBlend {blendOption=Linear}
    boundBone {boneName=dyn_bag_strap_jnt}
    componentAppliedLinearAccClamp=<2000 2000 2000>
    componentLinearAccScale=<0.3 0.3 0.3>
    componentLinearVelScale=<0.3 0.3 0.3>
    gravityOverride {z=-100}
    physicsBodyDefinitions [{
      boundBone {boneName=dyn_bag_strap_jnt}
      boxExtents=<19 10 10>
      constraintSetup {angularLimitsMax=<0 8 10>, angularLimitsMin=<0 -8 -50>, angularTarget=<1 -0.5 0>}
      localJointOffset=<1.5 0.5 0>
     }]
   }
   n10 = Anim Dynamics - Bone: dyn_axe_l_jnt(ComponentPose=@n9, Alpha=1, ExternalForce="0,0,0") {
    angularDampingOverride=0.9998
    angularSpringConstant=1500
    bAngularSpring=true
    bOverrideAngularDamping=true
    bUseGravityOverride=true
    gravityScale=0.35
    alphaBoolBlend {blendOption=Linear}
    boundBone {boneName=dyn_axe_l_jnt}
    componentAppliedLinearAccClamp=<1000 1000 1000>
    componentLinearAccScale=<0.2 0.2 0.4>
    componentLinearVelScale=<0.2 0.2 0.2>
    gravityOverride {z=-500}
    physicsBodyDefinitions [{
      boundBone {boneName=dyn_axe_l_jnt}
      boxExtents=<18 20 18>
      constraintSetup {angularLimitsMax=<0 0 15>, angularLimitsMin=<0 0 -15>, angularTarget=<1 0 0>}
      localJointOffset=<-1.5 -0.4 0>
     }]
   }
   n11 = Anim Dynamics - Bone: dyn_axe_r_jnt(ComponentPose=@n10, Alpha=1, ExternalForce="0,0,0") {
    angularDampingOverride=0.998
    angularSpringConstant=1500
    bAngularSpring=true
    bOverrideAngularDamping=true
    bUseGravityOverride=true
    gravityScale=0.35
    alphaBoolBlend {blendOption=Linear}
    boundBone {boneName=dyn_axe_r_jnt}
    componentAppliedLinearAccClamp=<1000 1000 1000>
    componentLinearAccScale=<0.2 0.2 0.2>
    componentLinearVelScale=<0.2 0.2 0.2>
    gravityOverride {z=-300}
    physicsBodyDefinitions [{
      boundBone {boneName=dyn_axe_r_jnt}
      boxExtents=<21 21 21>
      constraintSetup {angularLimitsMax=<0 0 15>, angularLimitsMin=<0 0 -15>, angularTarget=<1 0 0>}
      localJointOffset=<2 0 0>
     }]
   }
   n12 = Anim Dynamics - Bone: dyn_charm_01_jnt(ComponentPose=@n11, Alpha=1, ExternalForce="0,0,0") {
    angularDampingOverride=0.9998
    bOverrideAngularDamping=true
    bUseGravityOverride=true
    simulationSpace=RootRelative
    boundBone {boneName=dyn_charm_01_jnt}
    componentAppliedLinearAccClamp=<2000 2000 2000>
    componentLinearAccScale=<0.5 0.5 0.5>
    componentLinearVelScale=<0.5 0.5 0.5>
    gravityOverride {z=-600}
    physicsBodyDefinitions [{
      boundBone {boneName=dyn_charm_01_jnt}
      boxExtents=<8 8 8>
      constraintSetup {angularLimitsMax=<10 3 20>, angularLimitsMin=<-10 -10 -20>, angularTarget=<0 0 0>}
      localJointOffset=<1 0 0>
     }]
    relativeSpaceBone {boneName=dyn_axe_r_jnt}
   }
   n13 = Anim Dynamics - Bone: dyn_charm_02_jnt(ComponentPose=@n12, Alpha=1, ExternalForce="0,0,0") {
    angularDampingOverride=0.998
    bUseGravityOverride=true
    simulationSpace=RootRelative
    boundBone {boneName=dyn_charm_02_jnt}
    componentAppliedLinearAccClamp=<2000 2000 2000>
    componentLinearAccScale=<0.5 0.5 0.5>
    componentLinearVelScale=<0.5 0.5 0.5>
    gravityOverride {z=-600}
    physicsBodyDefinitions [{
      boundBone {boneName=dyn_charm_02_jnt}
      boxExtents=<8 8 8>
      constraintSetup {angularLimitsMax=<10 10 10>, angularLimitsMin=<-10 -20 -10>, angularTarget=<0 0 0>}
      localJointOffset=<1 0 0>
     }]
    relativeSpaceBone {boneName=dyn_axe_r_jnt}
   }
   n14 = Anim Dynamics - Bone: dyn_charm_03_jnt(ComponentPose=@n13, Alpha=1, ExternalForce="0,0,0") {
    angularDampingOverride=0.999
    bOverrideAngularDamping=true
    bUseGravityOverride=true
    simulationSpace=RootRelative
    boundBone {boneName=dyn_charm_03_jnt}
    componentLinearAccScale=<0.5 0.5 0.5>
    componentLinearVelScale=<0.5 0.5 0.5>
    gravityOverride {z=-600}
    physicsBodyDefinitions [{
      boundBone {boneName=dyn_charm_03_jnt}
      boxExtents=<5 5 5>
      constraintSetup {angularLimitsMax=<10 40 0>, angularLimitsMin=<-10 -40 0>, angularTarget=<0 0 0>}
      localJointOffset=<1 0 0>
     }]
    relativeSpaceBone {boneName=dyn_axe_r_jnt}
   }
   Output Pose(Result=Two Way Blend!(A=Component To Local(ComponentPose=@n7), B=Component To Local(ComponentPose=@n14), bAlphaBoolEnabled:bool=Get bIsInFrontEnd()) -> (Pose) {blendNode {alphaInputType=Bool}}) {layerGroup=None}
   Save cached pose 'InputPoseSaved'(Pose=Input Pose()) {cacheName=InputPoseSaved}
  ]}
 EventGraph {nodes [
   TryGetPawnOwner()
   Event BlueprintUpdateAnimation()
  ]}
]
variables []
